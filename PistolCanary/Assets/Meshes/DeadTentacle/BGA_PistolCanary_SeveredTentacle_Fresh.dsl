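; Fresh severed PistolCanary tentacle Blueprint whose large eye idles and watches nearby pawns, animated only on non-authority clients.
; A dynamic material instance receives eye X/Y offsets, driven at idle by the EyeDriver timeline with random pauses between plays.
; When a FortPawn enters LookAtCollision, a timeline eases the eye toward it, then a repeating timer maps its volume-local position to offsets.
; On exit, the timer clears and another timeline recenters the eye before idling.
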(event EventBeginPlay
  (Networking|SwitchHasAuthority
    (:Remote
      (bind _returnvalue (Rendering|Material|CreateDynamicMaterialInstance (Variables|Default|GetSM_PistolCanary_SeveredTentacle_Fresh) 2 "/PistolCanary/Assets/Materials/MI_Tentacle_Eye_Large_Inst.MI_Tentacle_Eye_Large_Inst" "None" :!class /Script/Engine.PrimitiveComponent))
      (Variables|Default|SetDynamicEyeMat _returnvalue)
      (CallFunction|EyeIdle)
      (|CallParentFunction))))

(event Custom|EyeIdle
  ; IdleAnim
  (if (not (Variables|Default|GetisInside))
    (bind (_direction _eyetrackx _eyetracky) (Timeline|EyeDriver PlayFromStart)
      (:Update
        (bind _output_get (Variables|Default|SetEye_X_Value _eyetrackx))
        (Rendering|Material|SetScalarParameterValue (Variables|Default|GetDynamicEyeMat) "Eye_X_Offset" _output_get :!class /Script/Engine.MaterialInstanceDynamic)
        (bind _output_get_2 (Variables|Default|SetEye_Y_Value _eyetracky))
        (Rendering|Material|SetScalarParameterValue (Variables|Default|GetDynamicEyeMat) "Eye_Y_Offset" _output_get_2 :!class /Script/Engine.MaterialInstanceDynamic))
      (:Finished
        (Utilities|FlowControl|Delay (Math|Random|RandomFloatinRange 1.0 3.0))
        (Timeline|EyeDriver PlayFromStart)))))

(event Custom|StopIdle
  ; IdleAnim
  (if (Variables|Default|GetisInside)
    (Timeline|EyeDriver Stop)))

(event Custom|FollowPlayer
  (bind (_pawnlocationinvolume_x _pawnlocationinvolume_y _pawnlocationinvolume_z) (Math|Vector|BreakVector (Variables|GetPawnLocationInVolume_X)))
  (bind _pawnlocationinvolume_x_2 (Math|Vector|BreakVector (Variables|GetPawnLocationInVolume_X)))
  ; EyeFollowPlayer
  (bind _output_get (Variables|Default|SetPawnLocationInVolume (Math|Transform|InverseTransformLocation (Transformation|GetWorldTransform (Variables|Default|GetLookAtCollision)) (Transformation|GetActorLocation (Variables|Default|GetOverlapPawn)))))
  (Development|PrintString (Utilities|String|ToString(Vector) _output_get))
  (bind _output_get_2 (Variables|Default|SetEye_X_Value (Math|Float|MapRangeClamped _pawnlocationinvolume_x_2 -1.25e+03 1.25e+03 0.4 -0.3)))
  (Rendering|Material|SetScalarParameterValue (Variables|Default|GetDynamicEyeMat) "Eye_X_Offset" _output_get_2 :!class /Script/Engine.MaterialInstanceDynamic)
  (bind _output_get_3 (Variables|Default|SetEye_Y_Value (Math|Float|MapRangeClamped _pawnlocationinvolume_z -7.5e+02 7.5e+02 0.3 -0.3)))
  (Rendering|Material|SetScalarParameterValue (Variables|Default|GetDynamicEyeMat) "Eye_Y_Offset" _output_get_3 :!class /Script/Engine.MaterialInstanceDynamic))

(event OnComponentBeginOverlap(LookAtCollision)
  (params (OverlappedComponent :type Primitive%20Component%20Object%20Reference) (OtherActor :type Actor%20Object%20Reference) (OtherComp :type Primitive%20Component%20Object%20Reference) (OtherBodyIndex :type Integer) (bFromSweep :type Boolean) (SweepResult :type Hit%20Result%20Structure%20%28by%20ref%29))
  (bind _pawnlocationinvolume_x (Math|Vector|BreakVector (Variables|GetPawnLocationInVolume_X)))
  (bind (_pawnlocationinvolume_x_2 _pawnlocationinvolume_y _pawnlocationinvolume_z) (Math|Vector|BreakVector (Variables|GetPawnLocationInVolume_X)))
  (Networking|SwitchHasAuthority
    (:Remote
      (bind _asfort_pawn (Utilities|Casting|CastToFortPawn OtherActor))
      (bind _output_get (Variables|Default|SetOverlapPawn _asfort_pawn))
      (if (Utilities|IsValid _output_get :!class /Script/Engine.KismetSystemLibrary)
        (Variables|Default|SetisInside true)
        (CallFunction|StopIdle)
        (Variables|Default|SetPawnLocationInVolume (Math|Transform|InverseTransformLocation (Transformation|GetWorldTransform (Variables|Default|GetLookAtCollision)) (Transformation|GetActorLocation (Variables|Default|GetOverlapPawn))))
        (if (Variables|Default|GetisInside)
          (bind (_direction _0_1) (Timeline|Timeline_0 PlayFromStart)
            (:Update
              (bind _output_get_2 (Variables|Default|SetEye_X_Value (Math|Interpolation|Ease "EaseInOut" _0_1 (Variables|Default|GetEye_X_Value) (Math|Float|MapRangeClamped _pawnlocationinvolume_x -1.25e+03 1.25e+03 0.4 -0.3) 3.0)))
              (Rendering|Material|SetScalarParameterValue (Variables|Default|GetDynamicEyeMat) "Eye_X_Offset" _output_get_2 :!class /Script/Engine.MaterialInstanceDynamic)
              (bind _output_get_3 (Variables|Default|SetEye_Y_Value (Math|Interpolation|Ease "EaseInOut" _0_1 (Variables|Default|GetEye_Y_Value) (Math|Float|MapRangeClamped _pawnlocationinvolume_z -7.5e+02 7.5e+02 0.3 -0.3) 3.0)))
              (Rendering|Material|SetScalarParameterValue (Variables|Default|GetDynamicEyeMat) "Eye_Y_Offset" _output_get_3 :!class /Script/Engine.MaterialInstanceDynamic))
            (:Finished
              (CallFunction|StopIdle)
              ; EyeFollowPlayer
              (bind _returnvalue (Utilities|Time|SetTimerbyEvent (delegate Custom|FollowPlayer) 0.2 true))
              (Variables|Default|SetTimer_FollowPlayer _returnvalue)))
          (else
            (Timeline|Timeline_0 Stop)))))))

(event OnComponentEndOverlap(LookAtCollision)
  (params (OverlappedComponent :type Primitive%20Component%20Object%20Reference) (OtherActor :type Actor%20Object%20Reference) (OtherComp :type Primitive%20Component%20Object%20Reference) (OtherBodyIndex :type Integer))
  (Networking|SwitchHasAuthority
    (:Remote
      (Variables|Default|SetisInside false)
      (Development|PrintString "goodbye")
      (Utilities|Time|ClearandInvalidateTimerbyHandle (Variables|Default|GetTimer_FollowPlayer))
      (if (Variables|Default|GetisInside)
        (bind (_direction _0_1) (Timeline|Timeline Stop)
          (:Update
            (bind _output_get (Variables|Default|SetEye_X_Value (Math|Interpolation|Ease "EaseInOut" _0_1 (Variables|Default|GetEye_X_Value) 0 3.0)))
            (Rendering|Material|SetScalarParameterValue (Variables|Default|GetDynamicEyeMat) "Eye_X_Offset" _output_get :!class /Script/Engine.MaterialInstanceDynamic)
            (bind _output_get_2 (Variables|Default|SetEye_Y_Value (Math|Interpolation|Ease "EaseInOut" _0_1 (Variables|Default|GetEye_Y_Value) 0 3.0)))
            (Rendering|Material|SetScalarParameterValue (Variables|Default|GetDynamicEyeMat) "Eye_Y_Offset" _output_get_2 :!class /Script/Engine.MaterialInstanceDynamic))
          (:Finished
            (CallFunction|EyeIdle)))
        (else
          (Timeline|Timeline PlayFromStart))))))

(fn OnRep_JumpCollisionLocation)

(fn ConstructionScript
  (|CallParentFunction))
