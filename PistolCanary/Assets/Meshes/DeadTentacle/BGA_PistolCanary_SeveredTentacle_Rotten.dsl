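; Blueprint actor for the rotten severed tentacle in PistolCanary, going by its name and folder.
; BeginPlay and the construction script only call their parent versions, so the graphs add no behavior of their own.
; An OnRep_JumpCollisionLocation replication callback is present with an empty body.
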(event EventBeginPlay
  (|CallParentFunction))

(fn OnRep_JumpCollisionLocation)

(fn ConstructionScript
  (|CallParentFunction))
