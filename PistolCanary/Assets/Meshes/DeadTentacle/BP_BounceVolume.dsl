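; Bounce volume Blueprint for a severed tentacle in Fortnite that springs overlapping pawns upward.
; When JumpCollision is overlapped, it caches the FortPawn and its player controller, then launches the pawn vertically using Final Bounce Jump Height, overriding Z velocity only.
; It suppresses the pawn's next fall damage and plays a bounce sound at its location.
; A short delay then leads to a leftover-looking debug print.
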(event EventBeginPlay
  (|CallParentFunction))

(event OnComponentBeginOverlap(JumpCollision)
  (params (OverlappedComponent :type Primitive%20Component%20Object%20Reference) (OtherActor :type Actor%20Object%20Reference) (OtherComp :type Primitive%20Component%20Object%20Reference) (OtherBodyIndex :type Integer) (bFromSweep :type Boolean) (SweepResult :type Hit%20Result%20Structure%20%28by%20ref%29))
  (bind _overlappawn (Variables|Default|GetOverlapPawn))
  (bind _asfort_pawn (Utilities|Casting|CastToFortPawn OtherActor))
  (bind _output_get (Variables|Default|SetOverlapPawn _asfort_pawn))
  (bind _asfort_player_controller (Utilities|Casting|CastToFortPlayerController (Pawn|GetController _output_get)))
  (Variables|Default|SetOverlapController _asfort_player_controller)
  (Character|LaunchCharacter _overlappawn (Math|Vector|MakeVector 0.0 0.0 (Variables|Default|GetFinal%20Bounce%20Jump%20Height)) false true)
  (Variables|/Script/FortniteGame.FortPawn|SetbIgnoreNextFallingDamage true _overlappawn)
  (Audio|PlaySoundatLocation "/PistolCanary_Event_Audio/Assets/PostNarrative/SC_PistolCanary_SeveredTentacle_Bounce.SC_PistolCanary_SeveredTentacle_Bounce" (Transformation|GetActorLocation _overlappawn))
  (Utilities|FlowControl|Delay 0.2)
  (Development|PrintString "DoneItAll!!"))

(fn ConstructionScript)
