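; BP_SlipVolume is a slip volume that applies a gameplay effect to pawns entering its JumpCollision component, apparently for ice-skating movement.
; Begin overlap casts the actor to FortPawn, stores it and its FortPlayerController, then applies the configurable effect class through the pawn's AbilitySystemComponent and prints "IceSkating!!!".
; End overlap is empty, so this graph never removes the effect.
; BeginPlay only calls its parent.
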(event EventBeginPlay
  (|CallParentFunction))

(event OnComponentBeginOverlap(JumpCollision)
  (params (OverlappedComponent :type Primitive%20Component%20Object%20Reference) (OtherActor :type Actor%20Object%20Reference) (OtherComp :type Primitive%20Component%20Object%20Reference) (OtherBodyIndex :type Integer) (bFromSweep :type Boolean) (SweepResult :type Hit%20Result%20Structure%20%28by%20ref%29))
  (bind _asfort_pawn (Utilities|Casting|CastToFortPawn OtherActor))
  (bind _output_get (Variables|Default|SetOverlapPawn _asfort_pawn))
  (bind _asfort_player_controller (Utilities|Casting|CastToFortPlayerController (Pawn|GetController _output_get)))
  (Variables|Default|SetOverlapController _asfort_player_controller)
  (GameplayEffects|ApplyGameplayEffectToSelf :self (Ability|GetAbilitySystemComponent (Variables|Default|GetOverlapPawn)) :GameplayEffectClass (Variables|Default|GetGameplay%20Effect%20Class) :Level 0.0)
  (Development|PrintString "IceSkating!!!"))

(event OnComponentEndOverlap(JumpCollision)
  (params (OverlappedComponent :type Primitive%20Component%20Object%20Reference) (OtherActor :type Actor%20Object%20Reference) (OtherComp :type Primitive%20Component%20Object%20Reference) (OtherBodyIndex :type Integer)))

(fn ConstructionScript)
